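# Niagara ribbon trail effect for the MarineCarve pickaxe, built from socket-driven anim trails in two emitters.
# MainTrail spawns two short-lived, narrowing ribbons with separate colors masked by spawn group, tinted by system time-of-day color and blended to darker, fainter versions by User.Variant.
# OllieGlowTrail adds two thin, camera-offset pink glow ribbons that flash briefly, strongest at night and invisible by day.
# It also derives a variant boolean from User.Variant.
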NiagaraSystem /BRCosmetics/Weapons/FORT_Melee/Pickaxe_MarineCarve/FX/NS_Pickaxe_MarineCarve_Trail
dependencies {
 usedDataInterfaces [&NiagaraDataInterfaceColorCurve, &NiagaraDataInterfaceCurve, &NiagaraDataInterfaceSocketReader]
 usedDynamicInputs []
 usedModules [
  &self:SystemSpawnScript.SystemScriptSource.SystemScriptGraph.NiagaraNodeAssignment_1.SetVariables_#1
  &SystemState
  &self:SystemSpawnScript.SystemScriptSource.SystemScriptGraph.NiagaraNodeAssignment_0.SetVariables_#2
 ]
}
emitters {
 MainTrail {
  emitterUpdateScript [
   &EmitterState()
   &FN_SpawnAnimTrail(Number of Anim Trails=2)
  ]
  particleSpawnScript [
   &FN_InitializeAnimTrailBoneIndecies()
   &FN_InitializeAnimTrailBoneIndecies(AnimTrail Selection=1, First Bone or Socket Index=2, Second Bone or Socket Index=3) {module=FN_InitializeAnimTrailBoneIndecies001}
   &FN_InitializeAnimTrail(Lifetime=0.25)
  ]
  particleUpdateScript [
   &ParticleState()
   &FN_UpdateAnimTrail()
   &ScaleColor(
    Scale Alpha=&Multiply_Float(
     A=&FNTimeOfDayScalar(Night Scale=2, Morning Scale=2, Day Scale=2, Evening Scale=2)
     B=&Lerp_Float(A=1, B=0.5, Alpha=$User.Variant)
    )
    Scale RGB=&MaskVectorBySpawnGroup(
     Masked Vector=&Multiply_Vector(
      A=&Lerp_Vector(
       A=&MakeVectorFromLinearColorRGB(Color=&ColorFromCurve(ColorCurve=DataInterface() {propertyValues {
           AlphaCurve {keys [<at 0 1 RCIM_Cubic>,<at 1 0>]}
           BlueCurve {keys [<at 0.4612 0.6979 RCIM_Cubic>]}
           GreenCurve {keys [<at 0.4612 0.1885 RCIM_Cubic>]}
           RedCurve {keys [<at 0.4612 0.109 RCIM_Cubic>]}
          }}))
       B=&MakeVectorFromLinearColorRGB(Color=&ColorFromCurve(ColorCurve=DataInterface() {propertyValues {
           AlphaCurve {keys [<at 0 1 RCIM_Cubic>,<at 0.9876 0>]}
           BlueCurve {keys [<at 0.4612 0.3698 RCIM_Cubic>]}
           GreenCurve {keys [<at 0.4612 0.0251 RCIM_Cubic>]}
           RedCurve {keys [<at 0.4612 0 RCIM_Cubic>]}
          }}))
       Alpha=$User.Variant
      )
      B=&Multiply_VectorByFloat(Vector=&MakeVectorFromLinearColorRGB(Color=$System.TOD), Float=0.25)
     )
     Passthrough Vector=&Multiply_Vector(
      A=&Lerp_Vector(
       A=&MakeVectorFromLinearColorRGB(Color=&ColorFromCurve(ColorCurve=DataInterface() {propertyValues {
           AlphaCurve {keys [<at 0 1 RCIM_Cubic>,<at 1 0>]}
           BlueCurve {keys [<at 0.4612 0.5633 RCIM_Cubic>]}
           GreenCurve {keys [<at 0.4612 0.249 RCIM_Cubic>]}
           RedCurve {keys [<at 0.4612 0.7969 RCIM_Cubic>]}
          }}))
       B=&MakeVectorFromLinearColorRGB(Color=&ColorFromCurve(ColorCurve=DataInterface() {propertyValues {
           AlphaCurve {keys [<at 0 1 RCIM_Cubic>,<at 0.9876 0>]}
           BlueCurve {keys [<at 0.4612 0 RCIM_Cubic>]}
           GreenCurve {keys [<at 0.4612 0 RCIM_Cubic>]}
           RedCurve {keys [<at 0.4612 0 RCIM_Cubic>]}
          }}))
       Alpha=$User.Variant
      )
      B=&Multiply_VectorByFloat(Vector=&MakeVectorFromLinearColorRGB(Color=$System.TOD), Float=0.25)
     )
    )
    ScaleA=true
    ScaleRGB=true
   )
   /SetVariables() {isSetParameters=true, module=SetVariables_#3}
   &ScaleRibbonWidth(Ribbon Width Scale=&FloatFromCurve(FloatCurve=DataInterface() {propertyValues {Curve {
        defaultValue=1
        keys [
         {interpMode=RCIM_Cubic, tangentMode=RCTM_User, time=0, value=1}
         {arriveTangent=-2.7093, interpMode=RCIM_Cubic, leaveTangent=-2.7093, tangentMode=RCTM_User, time=1, value=0}
        ]
       }}}))
  ]
  rendererClasses [&NiagaraRibbonRendererProperties]
  renderers [{rendererClass=&NiagaraRibbonRendererProperties}]
 }
 OllieGlowTrail {
  emitterUpdateScript [
   &EmitterState()
   &FN_SpawnAnimTrail(Number of Anim Trails=2)
   /SetVariables(Emitter.VariantBool:NiagaraBool=&SetBoolByFloatComparison(A=$User.Variant, B=1, Comparison Type=::NewEnumerator4)) {isSetParameters=true}
  ]
  particleSpawnScript [
   &FN_InitializeAnimTrailBoneIndecies(First Bone or Socket Index=4, Second Bone or Socket Index=4)
   &FN_InitializeAnimTrailBoneIndecies(AnimTrail Selection=1, First Bone or Socket Index=5, Second Bone or Socket Index=5) {module=FN_InitializeAnimTrailBoneIndecies001}
   &FN_InitializeAnimTrail(AnimTrailWidth=5, Color=<rgba 1 0.154 0.8449 1>, Lifetime=0.25)
   &CameraOffset(Camera Offset Amount=15)
  ]
  particleUpdateScript [
   &ParticleState()
   &FN_UpdateAnimTrail()
   &ScaleColor(
    Scale Alpha=&FloatFromCurve(
     FloatCurve=DataInterface() {propertyValues {Curve {
        defaultValue=1
        keys [
         <at 0 0 RCIM_Cubic>
         {interpMode=RCIM_Cubic, tangentMode=RCTM_User, time=0.1, value=1}
         {interpMode=RCIM_Cubic, tangentMode=RCTM_User, time=0.5, value=0}
        ]
       }}}
     Scale Curve=&FNTimeOfDayScalar(Night Scale=0.75, Morning Scale=0.3, Day Scale=0, Evening Scale=0.3)
    )
    ScaleA=true
    ScaleRGB=true
   )
   /SetVariables() {isSetParameters=true, module=SetVariables_#3}
   &ScaleRibbonWidth(Ribbon Width Scale=&FloatFromCurve(FloatCurve=DataInterface() {propertyValues {Curve {keys [<at 0.5 1>,<at 0.6 0>]}}}))
  ]
  rendererClasses [&NiagaraRibbonRendererProperties]
  renderers [{rendererClass=&NiagaraRibbonRendererProperties}]
 }
}
summary {
 systemName=NS_Pickaxe_MarineCarve_Trail
 emitters [
  {emitterName=MainTrail, rendererClasses [&NiagaraRibbonRendererProperties]}
  {emitterName=OllieGlowTrail, rendererClasses [&NiagaraRibbonRendererProperties]}
 ]
 userVariables [
  User.RibbonIDOffset {type=NiagaraInt32}
  User.Variant {type=NiagaraFloat, defaultValue=0, description=""}
 ]
}
