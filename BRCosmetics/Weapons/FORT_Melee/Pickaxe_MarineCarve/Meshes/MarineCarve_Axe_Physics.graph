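# Physics asset for the MarineCarve pickaxe cosmetic mesh, letting its five-bone dangling strap swing from the weapon.
# Two capsules cover the handle bone, and each dyn_strap bone carries one small capsule.
# Constraints chain the strap bones back to the handle with twist limited to zero and a fixed narrow first swing.
# The second swing limit tightens progressively from the wide handle joint to the strap tip.
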PhysicsAsset /BRCosmetics/Weapons/FORT_Melee/Pickaxe_MarineCarve/Meshes/MarineCarve_Axe_Physics
bodies {
 Melee_TwoHanded_Handle [
  {
   shapeType=Capsule
   length=31.7152
   radius=3.3
   shapeName=Melee_TwoHanded_Handle_capsule
   center=<0.082 -0.6342 16.0086>
   rotation=<rot -0.8996 -90.0343 1.8166>
  }
  {
   shapeType=Capsule
   length=11.8909
   radius=9.1623
   shapeName=Melee_TwoHanded_Handle_capsule
   center=<0.0001 0.2166 40.989>
   rotation=<rot -0.0084 0 -88.8558>
  }
 ]
 dyn_strap_01 [{shapeType=Capsule, length=1.9149, radius=1.6186, shapeName=dyn_strap_01_capsule, center=<2 0 0.5>, rotation=<rot 0 90 -90>}]
 dyn_strap_02 [{shapeType=Capsule, length=1.568, radius=1.449, shapeName=dyn_strap_02_capsule, center=<1.7 0 0.5>, rotation=<rot 0.97 -90 -90>}]
 dyn_strap_03 [{shapeType=Capsule, length=1.696, radius=1.309, shapeName=dyn_strap_03_capsule, center=<1.7 0 0.07>, rotation=<rot 0 -90 -90>}]
 dyn_strap_04 [{shapeType=Capsule, length=0.8494, radius=1.2706, shapeName=dyn_strap_04_capsule, center=<1 0.1 0.06>, rotation=<rot 0 -90 -90>}]
 dyn_strap_05 [{shapeType=Capsule, length=1.7901, radius=1.2199, shapeName=dyn_strap_05_capsule, center=<1.5 0.04 0.15>, rotation=<rot 1.5 -90 -90>}]
}
constraints [
 {bone1Name=dyn_strap_05, bone2Name=dyn_strap_04, swing1LimitDegrees=10, swing2LimitDegrees=10, twistLimitDegrees=0}
 {bone1Name=dyn_strap_04, bone2Name=dyn_strap_03, swing1LimitDegrees=10, swing2LimitDegrees=20, twistLimitDegrees=0}
 {bone1Name=dyn_strap_03, bone2Name=dyn_strap_02, swing1LimitDegrees=10, swing2LimitDegrees=30, twistLimitDegrees=0}
 {bone1Name=dyn_strap_02, bone2Name=dyn_strap_01, swing1LimitDegrees=10, swing2LimitDegrees=45, twistLimitDegrees=0}
 {bone1Name=dyn_strap_01, bone2Name=Melee_TwoHanded_Handle, swing1LimitDegrees=10, swing2LimitDegrees=90, twistLimitDegrees=0}
]
