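# Animation Blueprint for the MarineCarve pickaxe that adds secondary motion to its dangling strap.
# bRigidBodyEnabled selects between RigidBody simulation relative to the two-handed handle bone and a cheaper fallback: additive strap rotations feeding a rotation-limited Trail controller.
# In the front end, the owner's Frontend_Idle_Bypass curve blends the CustomFrontEndIdle slot over the result from base_strap_01 down.
# The EventGraph also sets MovingAlpha from jump/fall state, but the AnimGraph never reads it.
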AnimBlueprint /BRCosmetics/Weapons/FORT_Melee/Pickaxe_MarineCarve/Meshes/MarineCarve_Axe_AnimBP
target_skeleton=&/MarineCarve_Axe_Skeleton
graphs [
 AnimGraph {nodes [
   n1 = "Transform (Modify) Bone - Bone: dyn_strap_01"(
    ComponentPose=Local To Component(LocalPose=Use cached pose 'InputPoseCache'())
    Alpha=1
    Translation="0,0,0"
    Rotation="-60,0,0"
    Scale="1,1,1"
   ) {
    rotationMode=BMM_Additive
    rotationSpace=BCS_BoneSpace
    translationMode=BMM_Additive
    translationSpace=BCS_BoneSpace
    boneToModify {boneName=dyn_strap_01}
    rotation {pitch=-60}
   }
   n2 = "Transform (Modify) Bone - Bone: dyn_strap_02"(ComponentPose=@n1, Alpha=1, Translation="0,0,0", Rotation="-50,0,0", Scale="1,1,1") {
    rotationMode=BMM_Additive
    rotationSpace=BCS_BoneSpace
    translationMode=BMM_Additive
    translationSpace=BCS_BoneSpace
    boneToModify {boneName=dyn_strap_02}
    rotation {pitch=-50}
   }
   n3 = "Transform (Modify) Bone - Bone: dyn_strap_03"(ComponentPose=@n2, Alpha=1, Translation="0,0,0", Rotation="-30,0,0", Scale="1,1,1") {
    rotationMode=BMM_Additive
    rotationSpace=BCS_BoneSpace
    translationMode=BMM_Additive
    translationSpace=BCS_BoneSpace
    boneToModify {boneName=dyn_strap_03}
    rotation {pitch=-30}
   }
   n4 = Trail controller - Bone: dyn_strap_05(ComponentPose=@n3, Alpha=1) {
    bLimitRotation=true
    bShowTrailLocation=true
    bUsePlanarLimit=true
    chainLength=5
    maxDeltaTime=0.333
    rotationLimits [
     {limitMax=<0 50 0>, limitMin=<0 -50 20>}
     {limitMax=<0 15 0>, limitMin=<0 -40 20>}
     {limitMax=<0 20 0>, limitMin=<0 -50 20>}
     {limitMax=<0 25 0>, limitMin=<0 -60 20>}
     {limitMax=<0 30 0>, limitMin=<0 -70 20>}
    ]
    rotationOffsets [<0 0 0>, <0 0 0>, <0 0 0>, <0 0 0>, <0 0 0>]
    trailBone {boneName=dyn_strap_05}
    trailRelaxationSpeed {editorCurveData {keys [<at 0 10 RCIM_Linear>,<at 1 4.5219 RCIM_Linear>]}}
    alphaBoolBlend=~
   }
   n5 = Component To Local(ComponentPose=@n4)
   n6 = RigidBody(ComponentPose=Local To Component(LocalPose=Use cached pose 'InputPoseCache'())) {
    bOverrideWorldGravity=true
    overridePhysicsAsset=&/MarineCarve_Axe_Physics
    simulationSpace=BaseBoneSpace
    baseBoneRef {boneName=Melee_TwoHanded_Handle}
    overrideWorldGravity {z=-980}
    simSpaceSettings {
     maxAngularAcceleration=40
     maxAngularVelocity=8
     maxLinearAcceleration=3500
     maxLinearVelocity=700
     velocityScaleZ=0.6
     worldAlpha=1
     externalLinearDragV=<1 1 1>
     externalLinearVelocity=<1 1 1>
    }
   }
   n7 = Component To Local(ComponentPose=@n6)
   n8 = Get bRigidBodyEnabled()
   Output Pose(Result=Layered blend per bone(
     BasePose=Blend Poses by bool(BlendPose_0=@n7, BlendPose_1=@n5, bActiveValue=@n8)
     BlendPoses_0=Slot 'CustomFrontEndIdle'() {slotName=CustomFrontEndIdle}
     BlendWeights_0=Get Custom Idle Is Playing() -> (Custom Idle Is Playing:double)
    ) {layerSetup [{branchFilters [{boneName=base_strap_01}]}]}) {layerGroup=None}
   AnimGraphNode_SubInput_0 = Input Pose()
   Save cached pose 'InputPoseCache'(Pose=@AnimGraphNode_SubInput_0) {cacheName=InputPoseCache}
  ]}
 EventGraph {nodes [
   Event BlueprintUpdateAnimation()
   > n1 = Sequence()
   > n2 = Branch(Condition=OR Boolean!(A=Get bIsJumping(), B=Get bIsFalling(), C:bool) -> (ReturnValue))
   > Set MovingAlpha(MovingAlpha:double=0.5) -> (Output_Get:double)
   Set MovingAlpha(execute=@n2.else, MovingAlpha=1)
   n3 = GetOwnerFrontEndAnimInstance()
   Branch(execute=@n1.then_1, Condition=AND Boolean(A=Get bIsInFrontEnd(), B=IsValid(Object=@n3)))
   > Set Custom Idle Is Playing(Custom Idle Is Playing:double=GetCurveValue(self=@n3, CurveName=Frontend_Idle_Bypass)) -> (Output_Get:double)
   TryGetPawnOwner()
  ]}
]
variables [
 Custom Idle Is Playing {type=number}
 MovingAlpha {type=number}
]
